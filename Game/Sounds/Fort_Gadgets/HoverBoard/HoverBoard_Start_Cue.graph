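# HoverBoard_Start_Cue is a Sound Cue for the hoverboard gadget, named for its start sound.
# It has no root and no sound nodes, so as saved it plays no audio.
# It still carries the Fort_Ability_Radius attenuation, a short max distance and the Fort_WeaponMelee_PSM sound class.
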SoundCue /Game/Sounds/Fort_Gadgets/HoverBoard/HoverBoard_Start_Cue
root=null
nodes []
properties {
 attenuationSettings=/Game/Sounds/Attenuation/Fort_Ability_Radius
 maxDistance=1200
 soundClassObject=&Fort_WeaponMelee_PSM
 volumeMultiplier=1
}
